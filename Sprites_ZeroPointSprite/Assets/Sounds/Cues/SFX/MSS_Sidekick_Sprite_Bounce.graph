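# One-shot mono MetaSound Source for the sidekick Sprite's bounce sound effect, played through attenuation settings.
# The Bounce trigger, not OnPlay, compares BounceCount with the Bounces array length to route an index: either the count itself or the count minus (length − 1).
# That index picks one of the bounce waves for a Wave Player, whose completion finishes the source.
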MetaSoundSource /Sprites_ZeroPointSprite/Assets/Sounds/Cues/SFX/MSS_Sidekick_Sprite_Bounce
inputs [
 Bounce {type=Trigger}
 BounceCount {type=Int32}
 Bounces {type=WaveAsset:Array}
 UE.Source.OnPlay {type=Trigger}
]
interfaces [UE.OutputFormat.Mono:1.0, UE.Source.OneShot:1.0, UE.Source:1.1]
outputs [
 "UE.OutputFormat.Mono.Audio:0" {type=Audio}
 UE.Source.OneShot.OnFinished {type=Trigger}
]
pages [Default {
  input_defaults {
   Bounce=null
   BounceCount=4
   UE.Source.OnPlay=null
   Bounces [
    /Sprites_ZeroPointSprite/Assets/Sounds/Waves/SFX/SideKick_Sprite_Bounce_01
    /Sprites_ZeroPointSprite/Assets/Sounds/Waves/SFX/SideKick_Sprite_Bounce_02
    /Sprites_ZeroPointSprite/Assets/Sounds/Waves/SFX/SideKick_Sprite_Bounce_03
   ]
  }
  nodes [
   n1 = UE.Reroute(Value=^Bounces)
   n2 = Array.Num:WaveAsset:Array(Array=^Bounces)
   n3 = TriggerCompare.Clamp:Int32(Compare=^Bounce, A=^BounceCount, B=@n2.Num) {also {defaults {Type=3}}}
   TriggerRouteTriggerRouteInt322_0 = "TriggerRoute.Trigger Route (Int32, 2):Int32"(
    Value 1=UE.Reroute(Value=^BounceCount).Value
    Value 0=UE.Subtract:Int32(PrimaryOperand=^BounceCount, AdditionalOperands=UE.Subtract:Int32(PrimaryOperand=@n2.Num) {also {defaults {AdditionalOperands=1}}}.Out).Out
    Set 1=@n3.False
    Set 0=@n3.True
   )
   WavePlayer_0 = UE.Wave Player:Mono(
    Wave Asset=Array.Get:WaveAsset:Array(Array=UE.Reroute(Value=@n1.Value).Value, Trigger=@TriggerRouteTriggerRouteInt322_0.On Set, Index=@TriggerRouteTriggerRouteInt322_0.Value).Element
    Play=@TriggerRouteTriggerRouteInt322_0.On Set
   )
  ]
  links [{dst="@UE.Source.OneShot.OnFinished", src=WavePlayer_0.On Finished}, {dst="@UE.OutputFormat.Mono.Audio:0", src=WavePlayer_0.Out Mono}]
 }]
properties {
 assetImportData=&self:AssetImportData
 attenuationSettings=/Sprites_ZeroPointSprite/Assets/Sounds/Settings/AS_Sprite_SideKick_DamageBoost
 maxDistance=2500
 sampleRate=48000
 soundClassObject=&Fort_Gadgets_Items
}
