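# Sound cue for the bounce sound of the Zero Point Sprite sidekick, routed to the gadgets and items sound class.
# A Random node picks one of three equally weighted bounce waves and feeds a Modulator that holds volume constant, leaving pitch at its default range.
# It borrows the sidekick DamageBoost attenuation asset and overrides concurrency to allow a single instance with a retrigger time.
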SoundCue /Sprites_ZeroPointSprite/Assets/Sounds/Cues/SFX/SC_SideKick_Sprite_Bounce
root=Modulator_0
nodes [
 {id=Modulator_0, children [Random_0], properties {volumeMax=1, volumeMin=1}}
 {id=Random_0, children [WavePlayer_0,WavePlayer_1,WavePlayer_2], properties {weights [1,1,1]}}
 {id=WavePlayer_0, properties {soundWaveAssetPtr=/Sprites_ZeroPointSprite/Assets/Sounds/Waves/SFX/SideKick_Sprite_Bounce_01}}
 {id=WavePlayer_1, properties {soundWaveAssetPtr=/Sprites_ZeroPointSprite/Assets/Sounds/Waves/SFX/SideKick_Sprite_Bounce_02}}
 {id=WavePlayer_2, properties {soundWaveAssetPtr=/Sprites_ZeroPointSprite/Assets/Sounds/Waves/SFX/SideKick_Sprite_Bounce_03}}
]
properties {
 attenuationSettings=/Sprites_ZeroPointSprite/Assets/Sounds/Settings/AS_Sprite_SideKick_DamageBoost
 bOverrideConcurrency=true
 maxDistance=2500
 soundClassObject=&Fort_Gadgets_Items
 volumeMultiplier=0.7
 concurrencyOverrides {maxCount=1, retriggerTime=2}
}
