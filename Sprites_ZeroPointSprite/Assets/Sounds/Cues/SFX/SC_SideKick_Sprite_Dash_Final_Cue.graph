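# Sound cue for the dash sound effect of the Zero Point Sprite sidekick.
# A single wave player plays the final dash wave through a modulator that pins pitch at double speed with no random range.
# The cue routes to the gadgets sound class at reduced volume, with its own natural-sound attenuation override and a small concurrency cap.
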SoundCue /Sprites_ZeroPointSprite/Assets/Sounds/Cues/SFX/SC_SideKick_Sprite_Dash_Final_Cue
root=Modulator_0
nodes [
 {id=Modulator_0, children [WavePlayer_0], properties {pitchMax=2, pitchMin=2}}
 {id=WavePlayer_0, properties {soundWaveAssetPtr=/Sprites_ZeroPointSprite/Assets/Sounds/Waves/SFX/SideKick_Sprite_Dash_Final}}
]
properties {
 bOverrideAttenuation=true
 bOverrideConcurrency=true
 maxDistance=4000
 soundClassObject=&Fort_Gadgets
 volumeMultiplier=0.3
 attenuationOverrides {distanceAlgorithm=NaturalSound}
 concurrencyOverrides {maxCount=2}
}
