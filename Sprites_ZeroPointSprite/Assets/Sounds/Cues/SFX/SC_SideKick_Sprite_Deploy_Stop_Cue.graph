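# Sound cue for the stop of the Zero Point sprite sidekick's deploy, classed as a gadget sound.
# A mixer plays three layers at once: an end sound and the main stop sound, each through its own modulator, plus an unmodulated stop layer.
# The main stop wave is mixed far quieter than the other two.
# Overrides give it HRTF spatialization with natural falloff and a low concurrency cap.
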SoundCue /Sprites_ZeroPointSprite/Assets/Sounds/Cues/SFX/SC_SideKick_Sprite_Deploy_Stop_Cue
root=Mixer_0
nodes [
 {id=Mixer_0, children [Modulator_0,Modulator_1,WavePlayer_2], properties {inputVolume [1,0.1,1]}}
 {id=Modulator_0, children [WavePlayer_1]}
 {id=Modulator_1, children [WavePlayer_0]}
 {id=WavePlayer_0, properties {soundWaveAssetPtr=/Sprites_ZeroPointSprite/Assets/Sounds/Waves/SFX/SideKick_Sprite_Deploy_Stop}}
 {id=WavePlayer_1, properties {soundWaveAssetPtr=/Sprites_ZeroPointSprite/Assets/Sounds/Waves/SFX/SideKick_Sprite_End_01}}
 {id=WavePlayer_2, properties {soundWaveAssetPtr=/Sprites_ZeroPointSprite/Assets/Sounds/Waves/SFX/SideKick_Sprite_Deployed_StopLayer_01}}
]
properties {
 bOverrideAttenuation=true
 bOverrideConcurrency=true
 maxDistance=4000
 soundClassObject=&Fort_Gadgets
 volumeMultiplier=1
 attenuationOverrides {distanceAlgorithm=NaturalSound, spatializationAlgorithm=SPATIALIZATION_HRTF}
 concurrencyOverrides {maxCount=2}
}
