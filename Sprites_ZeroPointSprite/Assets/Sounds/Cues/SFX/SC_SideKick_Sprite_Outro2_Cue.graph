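# SoundCue for the second outro sound of the Zero Point Sprite sidekick.
# A single wave player feeds a modulator that applies a fixed upward pitch shift rather than a random range.
# It plays at reduced volume through the creature foley sound class and uses the close-range Sprite VO attenuation.
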SoundCue /Sprites_ZeroPointSprite/Assets/Sounds/Cues/SFX/SC_SideKick_Sprite_Outro2_Cue
root=Modulator_0
nodes [
 {id=Modulator_0, children [WavePlayer_0], properties {pitchMax=1.1, pitchMin=1.1}}
 {id=WavePlayer_0, properties {soundWaveAssetPtr=/Sprites_ZeroPointSprite/Assets/Sounds/Waves/SFX/SideKick_Sprite_Outro2_01}}
]
properties {
 attenuationSettings=/Sprites_ZeroPointSprite/Assets/Sounds/Settings/AS_Sprite_VO_Close
 maxDistance=4000
 soundClassObject=/Game/Sounds/FortSoundClassesAndModes/Fort_Creatures_Foley
 volumeMultiplier=0.4
}
