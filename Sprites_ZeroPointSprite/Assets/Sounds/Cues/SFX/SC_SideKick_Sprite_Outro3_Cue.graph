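# Sound cue for the Zero Point Sprite sidekick's third outro SFX.
# A Random node at the root picks with equal weight between two wave variations.
# It plays at reduced volume through the creature foley sound class, spatialized with the close sprite VO attenuation preset.
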SoundCue /Sprites_ZeroPointSprite/Assets/Sounds/Cues/SFX/SC_SideKick_Sprite_Outro3_Cue
root=Random_0
nodes [
 {id=Random_0, children [WavePlayer_0,WavePlayer_1], properties {weights [1,1]}}
 {id=WavePlayer_0, properties {soundWaveAssetPtr=/Sprites_ZeroPointSprite/Assets/Sounds/Waves/SFX/SideKick_Sprite_Outro3_02}}
 {id=WavePlayer_1, properties {soundWaveAssetPtr=/Sprites_ZeroPointSprite/Assets/Sounds/Waves/SFX/SideKick_Sprite_Outro3_01}}
]
properties {
 attenuationSettings=/Sprites_ZeroPointSprite/Assets/Sounds/Settings/AS_Sprite_VO_Close
 maxDistance=4000
 soundClassObject=/Game/Sounds/FortSoundClassesAndModes/Fort_Creatures_Foley
 volumeMultiplier=0.4
}
