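# Footstep Sound Cue for the sidekick Zero Point Sprite, played as creature foley with spatial attenuation and occlusion.
# A quality-level root sends its first two slots to pitched-down random steps mixed with ten random chimes.
# Its last two slots instead mix random water steps with four chimes, so quality level changes the sound's character.
# Per-owner concurrency caps overlapping steps at two, stopping the oldest.
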SoundCue /Sprites_ZeroPointSprite/Assets/Sounds/Cues/SFX/SC_SideKick_Sprite_Step
root=QualityLevel_0
nodes [
 {id=Mixer_1, children [Modulator_1,Random_4], properties {inputVolume [0.5,0.5]}}
 {id=Mixer_2, children [Random_3,Random_5], properties {inputVolume [1,0.5]}}
 {id=Modulator_1, children [Random_0], properties {pitchMax=0.6, pitchMin=0.6}}
 {id=QualityLevel_0, children [Mixer_1,Mixer_1,Mixer_2,Mixer_2]}
 {
  id=Random_0
  children [WavePlayer_0,WavePlayer_1,WavePlayer_2,WavePlayer_3,WavePlayer_4,WavePlayer_5,WavePlayer_6,WavePlayer_7]
  properties {weights [1,1,1,1,1,1,1,1]}
 }
 {id=Random_3, children [WavePlayer_20, WavePlayer_21, WavePlayer_22, WavePlayer_23], properties {weights [1,1,1,1]}}
 {
  id=Random_4
  children [
   WavePlayer_36
   WavePlayer_37
   WavePlayer_38
   WavePlayer_39
   WavePlayer_40
   WavePlayer_41
   WavePlayer_42
   WavePlayer_43
   WavePlayer_44
   WavePlayer_45
  ]
  properties {bShouldExcludeFromBranchCulling=true, weights [1,1,1,1,1,1,1,1,1,1]}
 }
 {id=Random_5, children [WavePlayer_46, WavePlayer_47, WavePlayer_48, WavePlayer_49], properties {weights [1,1,1,1]}}
 {id=WavePlayer_0, properties {soundWaveAssetPtr=/Sprites_ZeroPointSprite/Assets/Sounds/Waves/SFX/Sidekick_Sprite_Step_08}}
 {id=WavePlayer_1, properties {soundWaveAssetPtr=/Sprites_ZeroPointSprite/Assets/Sounds/Waves/SFX/Sidekick_Sprite_Step_07}}
 {id=WavePlayer_2, properties {soundWaveAssetPtr=/Sprites_ZeroPointSprite/Assets/Sounds/Waves/SFX/Sidekick_Sprite_Step_06}}
 {id=WavePlayer_20, properties {soundWaveAssetPtr=/Sprites_ZeroPointSprite/Assets/Sounds/Waves/SFX/Sprite_Water_Step_01}}
 {id=WavePlayer_21, properties {soundWaveAssetPtr=/Sprites_ZeroPointSprite/Assets/Sounds/Waves/SFX/Sprite_Water_Step_02}}
 {id=WavePlayer_22, properties {soundWaveAssetPtr=/Sprites_ZeroPointSprite/Assets/Sounds/Waves/SFX/Sprite_Water_Step_03}}
 {id=WavePlayer_23, properties {soundWaveAssetPtr=/Sprites_ZeroPointSprite/Assets/Sounds/Waves/SFX/Sprite_Water_Step_04}}
 {id=WavePlayer_3, properties {soundWaveAssetPtr=/Sprites_ZeroPointSprite/Assets/Sounds/Waves/SFX/Sidekick_Sprite_Step_05}}
 {id=WavePlayer_36, properties {soundWaveAssetPtr=/Sprites_ZeroPointSprite/Assets/Sounds/Waves/SFX/Sprites_Step_Chime_01}}
 {id=WavePlayer_37, properties {soundWaveAssetPtr=/Sprites_ZeroPointSprite/Assets/Sounds/Waves/SFX/Sprites_Step_Chime_02}}
 {id=WavePlayer_38, properties {soundWaveAssetPtr=/Sprites_ZeroPointSprite/Assets/Sounds/Waves/SFX/Sprites_Step_Chime_03}}
 {id=WavePlayer_39, properties {soundWaveAssetPtr=/Sprites_ZeroPointSprite/Assets/Sounds/Waves/SFX/Sprites_Step_Chime_04}}
 {id=WavePlayer_4, properties {soundWaveAssetPtr=/Sprites_ZeroPointSprite/Assets/Sounds/Waves/SFX/Sidekick_Sprite_Step_04}}
 {id=WavePlayer_40, properties {soundWaveAssetPtr=/Sprites_ZeroPointSprite/Assets/Sounds/Waves/SFX/Sprites_Step_Chime_05}}
 {id=WavePlayer_41, properties {soundWaveAssetPtr=/Sprites_ZeroPointSprite/Assets/Sounds/Waves/SFX/Sprites_Step_Chime_06}}
 {id=WavePlayer_42, properties {soundWaveAssetPtr=/Sprites_ZeroPointSprite/Assets/Sounds/Waves/SFX/Sprites_Step_Chime_07}}
 {id=WavePlayer_43, properties {soundWaveAssetPtr=/Sprites_ZeroPointSprite/Assets/Sounds/Waves/SFX/Sprites_Step_Chime_08}}
 {id=WavePlayer_44, properties {soundWaveAssetPtr=/Sprites_ZeroPointSprite/Assets/Sounds/Waves/SFX/Sprites_Step_Chime_09}}
 {id=WavePlayer_45, properties {soundWaveAssetPtr=/Sprites_ZeroPointSprite/Assets/Sounds/Waves/SFX/Sprites_Step_Chime_10}}
 {id=WavePlayer_46, properties {soundWaveAssetPtr=&Sprites_Step_Chime_01}}
 {id=WavePlayer_47, properties {soundWaveAssetPtr=&Sprites_Step_Chime_02}}
 {id=WavePlayer_48, properties {soundWaveAssetPtr=&Sprites_Step_Chime_03}}
 {id=WavePlayer_49, properties {soundWaveAssetPtr=&Sprites_Step_Chime_04}}
 {id=WavePlayer_5, properties {soundWaveAssetPtr=/Sprites_ZeroPointSprite/Assets/Sounds/Waves/SFX/Sidekick_Sprite_Step_03}}
 {id=WavePlayer_6, properties {soundWaveAssetPtr=/Sprites_ZeroPointSprite/Assets/Sounds/Waves/SFX/Sidekick_Sprite_Step_02}}
 {id=WavePlayer_7, properties {soundWaveAssetPtr=/Sprites_ZeroPointSprite/Assets/Sounds/Waves/SFX/Sidekick_Sprite_Step_01}}
]
properties {
 attenuationSettings=/Game/Sounds/Fort_Foley_Sounds/FoleyLibraries/Settings/Foley_Sprites_Attenuation_Basic_Att
 bOverrideConcurrency=true
 maxDistance=1536
 soundClassObject=/Game/Sounds/FortSoundClassesAndModes/Fort_Creatures_Foley
 volumeMultiplier=0.2
 attenuationOverrides {
  bEnableListenerFocus=true
  bEnableOcclusion=true
  falloffDistance=1024
  occlusionLowPassFilterFrequency=5000
  occlusionTraceChannel=ECC_GameTraceChannel11
  stereoSpread=0
  attenuationShapeExtents {x=512}
 }
 concurrencyOverrides {bLimitToOwner=true, maxCount=2, resolutionRule=StopOldest, volumeScaleAttackTime=0.05}
}
