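# Sound cue for the Zero Point Sprite sidekick's cook sound, layering cook SFX with sprite vocalizations.
# An items/gadgets sound-class branch mixes one of two random cook shots with a looping bed.
# A creature-foley branch plays reused inventory-landing VO with volume modulation, drawing on ten random variants for the first two quality levels and four for the last two.
# Custom attenuation and owner-limited concurrency, which stops the oldest of more than two instances, govern playback.
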SoundCue /Sprites_ZeroPointSprite/Assets/Sounds/Cues/SFX/SC_Sprite_SideKick_Cook
root=Mixer_2
nodes [
 {id=Mixer_0, children [Random_0,WavePlayer_4], properties {inputVolume [1,1]}}
 {id=Mixer_2, children [SoundClass_1,SoundClass_0], properties {inputVolume [1,1]}}
 {id=Modulator_0, children [QualityLevel_0], properties {volumeMin=0.6}}
 {id=QualityLevel_0, children [Random_1,Random_1,Random_2,Random_2]}
 {id=Random_0, children [WavePlayer_1,WavePlayer_2], properties {weights [1,1]}}
 {
  id=Random_1
  children [
   WavePlayer_0
   WavePlayer_3
   WavePlayer_5
   WavePlayer_6
   WavePlayer_7
   WavePlayer_10
   WavePlayer_11
   WavePlayer_12
   WavePlayer_13
   WavePlayer_14
  ]
  properties {bShouldExcludeFromBranchCulling=true, weights [1,1,1,1,1,1,1,1,1,1]}
 }
 {id=Random_2, children [WavePlayer_18, WavePlayer_17, WavePlayer_16, WavePlayer_15], properties {weights [1,1,1,1]}}
 {id=SoundClass_0, children [Modulator_0], properties {soundClassOverride=/Game/Sounds/FortSoundClassesAndModes/Fort_Creatures_Foley}}
 {id=SoundClass_1, children [Mixer_0], properties {soundClassOverride=&Fort_Gadgets_Items}}
 {id=WavePlayer_0, properties {soundWaveAssetPtr=/Sprites_ZeroPointSprite/Assets/Sounds/Waves/VO/VO_S33_ZeroPointSprite_Inventory_Land_010}}
 {id=WavePlayer_1, properties {soundWaveAssetPtr=/Sprites_ZeroPointSprite/Assets/Sounds/Waves/SFX/SideKick_Sprite_Cook_01}}
 {id=WavePlayer_10, properties {soundWaveAssetPtr=/Sprites_ZeroPointSprite/Assets/Sounds/Waves/VO/VO_S33_ZeroPointSprite_Inventory_Land_015}}
 {id=WavePlayer_11, properties {soundWaveAssetPtr=/Sprites_ZeroPointSprite/Assets/Sounds/Waves/VO/VO_S33_ZeroPointSprite_Inventory_Land_016}}
 {id=WavePlayer_12, properties {soundWaveAssetPtr=/Sprites_ZeroPointSprite/Assets/Sounds/Waves/VO/VO_S33_ZeroPointSprite_Inventory_Land_017}}
 {id=WavePlayer_13, properties {soundWaveAssetPtr=/Sprites_ZeroPointSprite/Assets/Sounds/Waves/VO/VO_S33_ZeroPointSprite_Inventory_Land_018}}
 {id=WavePlayer_14, properties {soundWaveAssetPtr=/Sprites_ZeroPointSprite/Assets/Sounds/Waves/VO/VO_S33_ZeroPointSprite_Inventory_Land_019}}
 {id=WavePlayer_15, properties {soundWaveAssetPtr=/Sprites_ZeroPointSprite/Assets/Sounds/Waves/VO/VO_S33_ZeroPointSprite_Inventory_Land_014}}
 {id=WavePlayer_16, properties {soundWaveAssetPtr=/Sprites_ZeroPointSprite/Assets/Sounds/Waves/VO/VO_S33_ZeroPointSprite_Inventory_Land_013}}
 {id=WavePlayer_17, properties {soundWaveAssetPtr=/Sprites_ZeroPointSprite/Assets/Sounds/Waves/VO/VO_S33_ZeroPointSprite_Inventory_Land_011}}
 {id=WavePlayer_18, properties {soundWaveAssetPtr=&VO_S33_ZeroPointSprite_Inventory_Land_010}}
 {id=WavePlayer_2, properties {soundWaveAssetPtr=/Sprites_ZeroPointSprite/Assets/Sounds/Waves/SFX/SideKick_Sprite_Cook_02}}
 {id=WavePlayer_3, properties {soundWaveAssetPtr=&VO_S33_ZeroPointSprite_Inventory_Land_011}}
 {id=WavePlayer_4, properties {bLooping=true, soundWaveAssetPtr=/Sprites_ZeroPointSprite/Assets/Sounds/Waves/SFX/SideKick_Sprite_Cook_Loop}}
 {id=WavePlayer_5, properties {soundWaveAssetPtr=/Sprites_ZeroPointSprite/Assets/Sounds/Waves/VO/VO_S33_ZeroPointSprite_Inventory_Land_012}}
 {id=WavePlayer_6, properties {soundWaveAssetPtr=&VO_S33_ZeroPointSprite_Inventory_Land_013}}
 {id=WavePlayer_7, properties {soundWaveAssetPtr=&VO_S33_ZeroPointSprite_Inventory_Land_014}}
]
properties {
 attenuationSettings=/Sprites_ZeroPointSprite/Assets/Sounds/Settings/AS_Sprite_Cook
 bOverrideConcurrency=true
 maxDistance=1000
 soundClassObject=&Fort_Creatures_Foley
 volumeMultiplier=0.2
 concurrencyOverrides {bLimitToOwner=true, maxCount=2, resolutionRule=StopOldest, volumeScaleAttackTime=0.05}
}
