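# Sound cue for the Zero Point Sprite sidekick's cook stop effect, played at low volume on the gadget items sound class.
# A Random node picks with equal weight between two "UnCook" wave variations.
# It uses the sprite cook attenuation and an overridden concurrency limit of two instances per owner, stopping the oldest.
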SoundCue /Sprites_ZeroPointSprite/Assets/Sounds/Cues/SFX/SC_Sprite_SideKick_Cook_Stop
root=Random_0
nodes [
 {id=Random_0, children [WavePlayer_1,WavePlayer_2], properties {weights [1,1]}}
 {id=WavePlayer_1, properties {soundWaveAssetPtr=/Sprites_ZeroPointSprite/Assets/Sounds/Waves/SFX/SideKick_Sprite_UnCook_01}}
 {id=WavePlayer_2, properties {soundWaveAssetPtr=/Sprites_ZeroPointSprite/Assets/Sounds/Waves/SFX/SideKick_Sprite_UnCook_02}}
]
properties {
 attenuationSettings=/Sprites_ZeroPointSprite/Assets/Sounds/Settings/AS_Sprite_Cook
 bOverrideConcurrency=true
 maxDistance=1000
 soundClassObject=&Fort_Gadgets_Items
 volumeMultiplier=0.14
 concurrencyOverrides {bLimitToOwner=true, maxCount=2, resolutionRule=StopOldest, volumeScaleAttackTime=0.05}
}
