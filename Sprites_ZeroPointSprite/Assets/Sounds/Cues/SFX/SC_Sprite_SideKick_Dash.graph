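# Sound cue for the dash effect of the Zero Point SideKick Sprite.
# A Random node picks one of three equally weighted dash waves and feeds a Modulator that varies pitch for each play.
# The cue uses the Fort_Traps sound class and references the Athena trap-activation attenuation asset.
# It overrides concurrency so that new instances are prevented once the limit is reached, rather than cutting off playing ones.
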SoundCue /Sprites_ZeroPointSprite/Assets/Sounds/Cues/SFX/SC_Sprite_SideKick_Dash
root=Modulator_10
nodes [
 {id=Modulator_10, children [Random_0], properties {pitchMax=1.1}}
 {id=Random_0, children [WavePlayer_0,WavePlayer_1,WavePlayer_2], properties {weights [1,1,1]}}
 {id=WavePlayer_0, properties {soundWaveAssetPtr=/Sprites_ZeroPointSprite/Assets/Sounds/Waves/SFX/SideKickSprite_Dash_03}}
 {id=WavePlayer_1, properties {soundWaveAssetPtr=/Sprites_ZeroPointSprite/Assets/Sounds/Waves/SFX/SideKickSprite_Dash_02}}
 {id=WavePlayer_2, properties {soundWaveAssetPtr=/Sprites_ZeroPointSprite/Assets/Sounds/Waves/SFX/SideKickSprite_Dash_01}}
]
properties {
 attenuationSettings=/Game/Sounds/Attenuation/Fort_Traps_Activate_Athena_Att
 bOverrideConcurrency=true
 maxDistance=3500
 soundClassObject=&Fort_Traps
 volumeMultiplier=1.4
 attenuationOverrides {falloffDistance=2100}
 concurrencyOverrides {resolutionRule=PreventNew}
}
