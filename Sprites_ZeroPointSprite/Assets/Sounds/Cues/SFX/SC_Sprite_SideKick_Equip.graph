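# Sound cue for the equip sound of a Sprite SideKick, routed to the creature foley sound class with Sprite foley attenuation.
# A Random node picks evenly among three equip wave variations and feeds a Modulator that varies volume.
# A concurrency override limits it to a couple of instances per owner and stops the oldest first.
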SoundCue /Sprites_ZeroPointSprite/Assets/Sounds/Cues/SFX/SC_Sprite_SideKick_Equip
root=Modulator_0
nodes [
 {id=Modulator_0, children [Random_2], properties {volumeMin=0.8}}
 {id=Random_2, children [WavePlayer_12, WavePlayer_13, WavePlayer_14], properties {weights [1,1,1]}}
 {id=WavePlayer_12, properties {soundWaveAssetPtr=/Sprites_ZeroPointSprite/Assets/Sounds/Waves/SFX/SideKick_Sprite_Equip_01}}
 {id=WavePlayer_13, properties {soundWaveAssetPtr=/Sprites_ZeroPointSprite/Assets/Sounds/Waves/SFX/SideKick_Sprite_Equip_02}}
 {id=WavePlayer_14, properties {soundWaveAssetPtr=/Sprites_ZeroPointSprite/Assets/Sounds/Waves/SFX/SideKick_Sprite_Equip_03}}
]
properties {
 attenuationSettings=/Game/Sounds/Fort_Foley_Sounds/FoleyLibraries/Settings/Foley_Sprites_Attenuation_Basic_Att
 bOverrideConcurrency=true
 maxDistance=1536
 soundClassObject=/Game/Sounds/FortSoundClassesAndModes/Fort_Creatures_Foley
 volumeMultiplier=0.5
 concurrencyOverrides {bLimitToOwner=true, maxCount=2, resolutionRule=StopOldest, volumeScaleAttackTime=0.05}
}
