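# Sound cue for the Zero Point Sprite sidekick throw sound effect, currently with an empty node graph and no root, so it plays no audio.
# It still carries playback settings: a shared Sprite cook attenuation, the Fortnite creatures foley sound class and reduced volume.
# A concurrency override limits overlapping instances per owner, stopping the oldest.
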SoundCue /Sprites_ZeroPointSprite/Assets/Sounds/Cues/SFX/SC_Sprite_SideKick_Throw
root=null
nodes []
properties {
 attenuationSettings=/Sprites_ZeroPointSprite/Assets/Sounds/Settings/AS_Sprite_Cook
 bOverrideConcurrency=true
 maxDistance=1000
 soundClassObject=/Game/Sounds/FortSoundClassesAndModes/Fort_Creatures_Foley
 volumeMultiplier=0.5
 concurrencyOverrides {bLimitToOwner=true, maxCount=2, resolutionRule=StopOldest, volumeScaleAttackTime=0.05}
}
