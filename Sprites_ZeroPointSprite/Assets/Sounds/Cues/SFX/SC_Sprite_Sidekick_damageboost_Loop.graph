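# Looping sound cue for the Sprite Sidekick damage boost, classed as a gadget item sound.
# A single looping wave player plays a deployed sidekick sprite loop wave through an Enveloper root.
# The Enveloper's volume curve drops to zero almost immediately, then rises back to full, acting as a short fade-in.
# Dedicated damage boost attenuation settings handle spatial falloff.
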SoundCue /Sprites_ZeroPointSprite/Assets/Sounds/Cues/SFX/SC_Sprite_Sidekick_damageboost_Loop
root=Enveloper_1
nodes [
 {
  id=Enveloper_1
  children [WavePlayer_2]
  properties {pitchCurve {editorCurveData {keys [{time=0}]}}, volumeCurve {editorCurveData {keys [<at 0 1>,<at 0.0001 0>,<at 1 1>]}}}
 }
 {
  id=WavePlayer_2
  properties {bLooping=true, soundWaveAssetPtr=/Sprites_ZeroPointSprite/Assets/Sounds/Waves/SFX/SideKickSprite_Deployed_Loop_B_01}
 }
]
properties {
 attenuationSettings=/Sprites_ZeroPointSprite/Assets/Sounds/Settings/AS_Sprite_SideKick_DamageBoost
 maxDistance=2500
 soundClassObject=&Fort_Gadgets_Items
 volumeMultiplier=1
}
