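# Sound cue for the Sprite sidekick's deployed burst effect in the Zero Point Sprite plugin.
# A Random node picks one of two burst wave variations with equal weight.
# It plays at boosted volume under the gadgets and items sound class, using a Sprite sidekick attenuation asset named for damage boost.
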SoundCue /Sprites_ZeroPointSprite/Assets/Sounds/Cues/SFX/SC_Sprite_Sidekick_deployed_Burst
root=Random_0
nodes [
 {id=Random_0, children [WavePlayer_3,WavePlayer_4], properties {weights [1,1]}}
 {id=WavePlayer_3, properties {soundWaveAssetPtr=/Sprites_ZeroPointSprite/Assets/Sounds/Waves/SFX/SideKick_Sprite_Deployed_Burst_01}}
 {id=WavePlayer_4, properties {soundWaveAssetPtr=/Sprites_ZeroPointSprite/Assets/Sounds/Waves/SFX/SideKick_Sprite_Deployed_Burst_02}}
]
properties {
 attenuationSettings=/Sprites_ZeroPointSprite/Assets/Sounds/Settings/AS_Sprite_SideKick_DamageBoost
 maxDistance=2500
 soundClassObject=&Fort_Gadgets_Items
 volumeMultiplier=2
}
