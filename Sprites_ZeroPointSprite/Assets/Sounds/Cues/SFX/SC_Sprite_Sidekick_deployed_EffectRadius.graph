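# Sound cue for the looping effect-radius sound of a deployed Zero Point Sprite sidekick.
# A single looping wave player feeds a Modulator node, which is the cue root.
# It plays at reduced volume under the Fort gadget items sound class, using attenuation settings named for the sidekick damage boost.
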SoundCue /Sprites_ZeroPointSprite/Assets/Sounds/Cues/SFX/SC_Sprite_Sidekick_deployed_EffectRadius
root=Modulator_0
nodes [
 {id=Modulator_0, children [WavePlayer_0]}
 {
  id=WavePlayer_0
  properties {bLooping=true, soundWaveAssetPtr=/Sprites_ZeroPointSprite/Assets/Sounds/Waves/SFX/SideKick_Sprite_Deployed_EffectRadius}
 }
]
properties {
 attenuationSettings=/Sprites_ZeroPointSprite/Assets/Sounds/Settings/AS_Sprite_SideKick_DamageBoost
 maxDistance=2500
 soundClassObject=&Fort_Gadgets_Items
 volumeMultiplier=0.5
}
