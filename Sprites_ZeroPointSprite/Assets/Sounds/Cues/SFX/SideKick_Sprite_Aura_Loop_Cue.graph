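# Sound cue for the looping aura of the Zero Point Sprite sidekick, routed through the gadgets sound class.
# A Mixer blends a looping aura wave with quieter electric spark one-shots, chosen at equal weight by a Random node behind a looped, variable Delay for irregular crackle.
# A root Modulator feeds the output, which is heavily attenuated in volume, uses DamageBoost sidekick attenuation settings and overrides concurrency to limit simultaneous instances.
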SoundCue /Sprites_ZeroPointSprite/Assets/Sounds/Cues/SFX/SideKick_Sprite_Aura_Loop_Cue
root=Modulator_0
nodes [
 {id=Delay_2, children [Random_0], properties {delayMax=2}}
 {id=Looping_0, children [Delay_2]}
 {id=Mixer_1, children [Looping_0,WavePlayer_0], properties {inputVolume [0.2,0.5]}}
 {id=Modulator_0, children [Mixer_1]}
 {id=Random_0, children [WavePlayer_2,WavePlayer_3,WavePlayer_4,WavePlayer_5,WavePlayer_6,WavePlayer_7], properties {weights [1,1,1,1,1,1]}}
 {id=WavePlayer_0, properties {bLooping=true, soundWaveAssetPtr=/Sprites_ZeroPointSprite/Assets/Sounds/Waves/SFX/SideKick_Sprite_aura_loop}}
 {id=WavePlayer_2, properties {soundWaveAssetPtr=/Game/Athena/Sounds/Quests/EnergyCloud/Spot_Electric_Spark_Arc_Small_06}}
 {id=WavePlayer_3, properties {soundWaveAssetPtr=/Game/Athena/Sounds/Quests/EnergyCloud/Spot_Electric_Spark_Arc_Small_05}}
 {id=WavePlayer_4, properties {soundWaveAssetPtr=/Game/Athena/Sounds/Quests/EnergyCloud/Spot_Electric_Spark_Arc_Small_04}}
 {id=WavePlayer_5, properties {soundWaveAssetPtr=/Game/Athena/Sounds/Quests/EnergyCloud/Spot_Electric_Spark_Arc_Small_03}}
 {id=WavePlayer_6, properties {soundWaveAssetPtr=/Game/Athena/Sounds/Quests/EnergyCloud/Spot_Electric_Spark_Arc_Small_02}}
 {id=WavePlayer_7, properties {soundWaveAssetPtr=/Game/Athena/Sounds/Quests/EnergyCloud/Spot_Electric_Spark_Arc_Small_01}}
]
properties {
 attenuationSettings=/Sprites_ZeroPointSprite/Assets/Sounds/Settings/AS_Sprite_SideKick_DamageBoost
 bOverrideConcurrency=true
 maxDistance=2500
 soundClassObject=&Fort_Gadgets
 volumeMultiplier=0.1
 concurrencyOverrides {maxCount=2}
}
